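; Blueprint for the Hightower smoke signal effect in an Apollo environment set, exposing a point-of-interest identifier.
; A custom EnableNearParticles event waits one frame with a zero-length Delay, then activates the near particle system with reset.
; The delay works around activation right after the effect is unhidden in game.
; GetPOIIdentifierValue returns the POI_Identifier variable, and the construction script is empty.
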(event Custom|EnableNearParticles
  ;[0 0 0] fix for activating, one frame delay after being set unhidden in game
  (Utilities|FlowControl|Delay 0.0)
  (Components|Activation|Activate (Variables|Default|GetNear%20Particle%20system) true))

(fn GetPOIIdentifierValue
  (return (Variables|Default|GetPOI_Identifier)))

(fn ConstructionScript)
